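# Post-process domain material that bakes an attribute value into one texel-wide column of a texture, apparently for Asteria tree foliage.
# TextureWidth and Index define a U range, and two If nodes output Value inside it and zero elsewhere.
# That mask is multiplied by the Vector3 parameter and added to a sample of Attribute_Bake_RT2, layering the column over existing contents.
# Negative emissive is allowed, so signed values are supported.
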Material /Game/Environments/Asteria/Foliage/Trees/MasterMaterials/BakeArrayTexture/M_Attribute_Bake
graph {
 properties {
  bAllowNegativeEmissiveColor=true
  blendableOutputAlpha=true
  materialDomain=MD_PostProcess
  pixelDepthOffsetMode=PDOM_Legacy
  usedShadingModels=Default Lit
 }
 nodes [
  n1 = ComponentMask(None=TextureCoordinate()) {r=true}
  n2 = Divide(A=Constant() {r=1}, B=ScalarParameter() {defaultValue=128, parameterName=TextureWidth})
  n3 = Multiply(A=@n2, B=ScalarParameter() {defaultValue=7, parameterName=Index})
  n4 = Add(A=@n2, B=@n3)
  n5 = Constant()
  n6 = ScalarParameter() {defaultValue=0.1, parameterName=Value}
  n7 = If(A=@n1, B=@n3, "A > B"=@n6, "A == B"=@n6, "A < B"=@n5)
  n8 = TextureSample() {samplerType=SAMPLERTYPE_LinearColor, texture=&/Attribute_Bake_RT2}
  Add(
   A=Multiply(A=VectorParameter() {parameterName=Vector3, defaultValue=<rgba 1 1 1 1>}.RGB, B=If(A=@n1, B=@n4, "A > B"=@n5, "A == B"=@n5, "A < B"=@n7))
   B=@n8.RGB
  )
  Constant() {r=1}
 ]
}
