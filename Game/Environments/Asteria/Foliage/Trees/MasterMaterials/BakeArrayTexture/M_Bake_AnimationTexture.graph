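# Unlit post-process material for baking foliage animation textures, defaulting to a Birch8 position/orientation map, with negative emissive and alpha output enabled.
# It remaps sampled RGB and alpha by dividing by Range and adding Bias, clamping only alpha.
# Separate branches snap UVs to SizeX by SizeY texel centers via a NearFilter blend, and convert RGB from linear to sRGB.
# These branches show no links to material outputs or the sampler's UVs.
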Material /Game/Environments/Asteria/Foliage/Trees/MasterMaterials/BakeArrayTexture/M_Bake_AnimationTexture
graph {
 properties {
  bAllowNegativeEmissiveColor=true
  blendableLocation=BL_SceneColorAfterDOF
  blendableOutputAlpha=true
  materialDomain=MD_PostProcess
  pixelDepthOffsetMode=PDOM_Legacy
  shadingModel=MSM_Unlit
  usedShadingModels=Unlit
 }
 nodes [
  n1 = TextureSampleParameter2D() {
   parameterName=Texture
   samplerType=SAMPLERTYPE_LinearColor
   texture=/Game/Environments/Asteria/Foliage/Trees/Birch/Textures/Animation/Birch8/Birch8Anim_SKEL_ANIM_Position_ORIENT_V01
  }
  n2 = ScalarParameter() {defaultValue=1, parameterName=Range}
  n3 = ScalarParameter() {parameterName=Bias}
  Add(
   A=Divide(
    A=@n1.RGB
    B=MaterialFunctionCall(X=@n2, Y=@n2, Z=@n2) {materialFunction=&MakeFloat3}.Result
   ) {constB=256}
   B=MaterialFunctionCall(X=@n3, Y=@n3, Z=@n3) {materialFunction=&MakeFloat3}.Result
  ) {constB=0.5}
  Clamp(None=Add(A=Divide(A=@n1.A, B=@n2) {constB=256}, B=@n3) {constB=0.5})
  n4 = TextureCoordinate()
  n5 = MaterialFunctionCall(Float2=@n4) {materialFunction=&BreakOutFloat2Components}
  n6 = ScalarParameter() {defaultValue=64, parameterName=SizeX}
  n7 = Add(A=Floor(None=Multiply(A=@n5.R, B=@n6))) {constB=0.5}
  n8 = ScalarParameter() {defaultValue=26, parameterName=SizeY}
  n9 = Add(A=Floor(None=Multiply(A=@n5.G, B=@n8))) {constB=0.5}
  LinearInterpolate(A=@n4, B=AppendVector(A=Divide(A=@n7, B=@n6), B=Divide(A=@n9, B=@n8)), Alpha=ScalarParameter() {defaultValue=1, parameterName=NearFilter})
  n10 = MaterialFunctionCall(Float3=@n1.RGB) {materialFunction=&BreakOutFloat3Components}
  MaterialFunctionCall(
   X=MaterialFunctionCall!(Linear=@n10.R) {materialFunction=/Engine/Functions/Engine_MaterialFunctions01/ImageAdjustment/LinearTosRGB}.sRGB
   Y=MaterialFunctionCall(Linear=@n10.G) {materialFunction=&LinearTosRGB}.sRGB
   Z=MaterialFunctionCall(Linear=@n10.B) {materialFunction=&LinearTosRGB}.sRGB
  ) {materialFunction=&MakeFloat3}
 ]
}
