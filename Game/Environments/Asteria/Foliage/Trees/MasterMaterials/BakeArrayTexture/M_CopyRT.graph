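# M_CopyRT is a post-process material in the Asteria tree BakeArrayTexture folder, apparently for copying a baked render target.
# It holds a linear-color sample of Attribute_Bake_RT and a constant of one, with no visible wiring to the material outputs.
# It allows negative emissive color and enables blendable output alpha, so it can output signed values and an alpha channel.
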Material /Game/Environments/Asteria/Foliage/Trees/MasterMaterials/BakeArrayTexture/M_CopyRT
graph {
 properties {
  bAllowNegativeEmissiveColor=true
  blendableOutputAlpha=true
  materialDomain=MD_PostProcess
  pixelDepthOffsetMode=PDOM_Legacy
  usedShadingModels=Default Lit
 }
 nodes [
  Constant() {r=1}
  TextureSample() {samplerType=SAMPLERTYPE_LinearColor, texture=&/Attribute_Bake_RT}
 ]
}
