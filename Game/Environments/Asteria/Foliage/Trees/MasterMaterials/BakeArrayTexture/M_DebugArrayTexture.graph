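# M_DebugArrayTexture is a material for inspecting texture arrays, stored with the Asteria tree foliage's array baking master materials.
# A texture coordinate is appended with the scalar Index parameter to form the UVs, so Index picks which slice of the Array parameter is sampled.
# Array defaults to the Asteria tree 1D bind pose texture array and is sampled as linear color.
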Material /Game/Environments/Asteria/Foliage/Trees/MasterMaterials/BakeArrayTexture/M_DebugArrayTexture
graph {
 properties {pixelDepthOffsetMode=PDOM_Legacy, usedShadingModels=Default Lit}
 nodes [
  Constant()
  Constant()
  TextureSampleParameter2DArray!(UVs=AppendVector(A=TextureCoordinate(), B=ScalarParameter() {parameterName=Index}), Apply View MipBias) {
   parameterName=Array
   samplerType=SAMPLERTYPE_LinearColor
   texture=/Game/Environments/Asteria/Foliage/Trees/MasterMaterials/TextureArrays/Asteria_Tree_BindPose1D_TA
  }
 ]
}
