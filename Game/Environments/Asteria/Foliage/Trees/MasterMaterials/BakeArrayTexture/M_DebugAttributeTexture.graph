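# Debug material that displays values from the Asteria tree attribute texture array as on-screen digits.
# Index and ArrayIndex parameters pick the column and slice, sampled at V = 0.5 in linear color, and the RGB feeds DebugFloat3Values.
# U is Index plus a small offset, divided by a hardcoded width of 128; the offset is half of 1/128, so it lands near the texel edge rather than its center.
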Material /Game/Environments/Asteria/Foliage/Trees/MasterMaterials/BakeArrayTexture/M_DebugAttributeTexture
graph {
 properties {pixelDepthOffsetMode=PDOM_Legacy, usedShadingModels=Default Lit}
 nodes [
  Constant()
  n1 = Divide(A=Constant() {r=1}, B=Constant() {desc=TextureWidth, r=128})
  n2 = Multiply(A=@n1, B=Add(A=ScalarParameter() {parameterName=Index}, B=Multiply(A=@n1) {constB=0.5}) {constB=0.5})
  n3 = Constant() {r=0.5}
  n4 = TextureSample(UVs=AppendVector(A=AppendVector(A=@n2, B=@n3), B=ScalarParameter() {parameterName=ArrayIndex})) {
   samplerType=SAMPLERTYPE_LinearColor
   texture=/Game/Environments/Asteria/Foliage/Trees/MasterMaterials/TextureArrays/Asteria_Tree_AtrributeArray
  }
  MaterialFunctionCall!(Vector3=@n4.RGB, MaximumNumberOfDigits, UVs, DebugTextLocation RG_UpperRight BA_LowerLeft, Component Spacing) {materialFunction=/Engine/Functions/Engine_MaterialFunctions02/Utility/DebugFloat3Values}
 ]
}
